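# Unlit translucent sprite material for snowflake particles in skydiving contrails, picking a 2x2 flipbook frame through a dynamic parameter.
# Opacity multiplies snowflake shape by distorted fire noise, a radial caustics-warped erosion mask and particle alpha, with optional depth fade.
# Emissive blends time-of-day blues with ripple and normal-derived detail, adding glow from the difference of two offset erosion masks.
# Mobile paths drop distortion, erosion, glow, depth fade and normal mapping.
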Material /FNGameplayCues/Athena/Effects/Materials/Skydiving/M_Contrail_Snowflakes
graph {
 properties {
  bUsedWithNiagaraSprites=true
  bUsedWithParticleSprites=true
  blendMode=BLEND_Translucent
  pixelDepthOffsetMode=PDOM_Legacy
  refractionMethod=RM_IndexOfRefraction
  shadingModel=MSM_Unlit
  translucencyLightingMode=TLM_VolumetricPerVertexDirectional
  usedShadingModels=Unlit
 }
 nodes [
  Constant() {r=1}
  n1 = Constant() {r=2}
  n2 = Constant() {r=2}
  n3 = DynamicParameter() {defaultValue=<rgba 0 0 1 1>}
  n4 = Power(Base=MaterialFunctionCall!(
    Texture=TextureObject() {texture=/Game/Effects/Fort_Effects/Textures/Snowflake/T_Snowflakes_F}
    UVs
    SubImages=AppendVector(A=@n1, B=@n2)
    Frame=Round(None=@n3.Param4)
   ) {materialFunction=/Engine/Functions/Engine_MaterialFunctions03/Texturing/SubUV_Function}.RGB)
  n5 = ScalarParameter() {defaultValue=100, parameterName=Depth_Fade}
  n6 = TextureCoordinate() {uTiling=0.4, vTiling=0.32}
  n7 = TextureCoordinate() {uTiling=0.5, vTiling=0.42}
  n8 = ComponentMask(None=TextureSample(UVs=Panner(Coordinate=@n7) {speedX=-0.1, speedY=-0.22}) {samplerType=SAMPLERTYPE_Normal, texture=&T_Weap_Distort_N_02}.RGB) {g=true, r=true}
  n9 = Panner(Coordinate=Add(A=@n6, B=Multiply(A=@n8) {constB=0.15})) {speedY=0.14}
  n10 = Panner(Coordinate=@n6) {speedY=0.14}
  n11 = Multiply(
   A=TextureSample(UVs=ShadingPathSwitch(Default=@n9, Mobile=@n10)) {texture=/Game/Global/Textures/Default/Noise/T_FireNoise_01}.RGB
   B=Constant() {r=3}
  )
  n12 = TextureSample(UVs=Panner() {speedX=-0.1, speedY=-0.22}) {texture=&T_ShieldCaustics}
  n13 = Add(
   A=MaterialFunctionCall() {materialFunction=&VectorToRadialValue}.Radial Coordinates
   B=Add(B=Multiply(A=@n12.R) {constB=0.052})
  )
  n14 = Constant() {r=1}
  n15 = Time()
  n16 = DynamicParameter() {defaultValue=<rgba 0 0 1 1>}
  n17 = Panner(Coordinate=@n13, Time=@n14, Speed=AppendVector(A=Multiply(A=@n15) {constB=0.1}, B=@n16.Param1))
  n18 = OneMinus(None=Multiply(A=TextureSample(UVs=@n17) {texture=/Game/Effects/Fort_Effects/Textures/Fire/T_Fire_Mask}.R))
  n19 = Constant() {r=1}
  n20 = Multiply(A=Multiply(A=Multiply(A=@n11, B=@n4), B=ShadingPathSwitch(Default=@n18, Mobile=@n19)), B=ParticleColor().A)
  n21 = StaticSwitchParameter(True=ShadingPathSwitch(Default=DepthFade(Opacity=@n20, FadeDistance=@n5), Mobile=@n20), False=@n20) {parameterName=Use_Depth Fade}
  If(A=ScalarParameter() {parameterName=USE_OPACITY}, "A > B"=Multiply(A=@n4, B=ParticleColor().A), "A == B"=@n21, "A < B"=@n21)
  n22 = ParticleColor()
  n23 = Constant() {r=1}
  n24 = Constant3Vector() {constant=<rgba 0.1826 0.5422 0.974 1>}
  n25 = Constant3Vector() {constant=<rgba 0.3347 0.5908 0.974 1>}
  n26 = MaterialFunctionCall(Color_Morning=@n24, Color_Day=@n24, Color_Evening=@n25, Color_Night=@n25) {materialFunction=&Color_Time_of_Day}
  n27 = Constant3Vector() {constant=<rgba 0.8604 0.928 0.9896 0>}
  n28 = TextureSample(UVs=TextureCoordinate() {uTiling=0.42, vTiling=0.58}) {texture=/Game/Effects/Fort_Effects/Textures/Tile/T_RippleTile_D}
  n29 = ScalarParameter() {defaultValue=2, parameterName=Confetti0_U_Repeats}
  n30 = ScalarParameter() {defaultValue=2, parameterName=Confetti0_V_Repeats}
  n31 = DynamicParameter() {defaultValue=<rgba 0 0 1 1>}
  n32 = TextureCoordinate()
  n33 = LinearInterpolate(
   A=Constant3Vector() {constant=<rgba 0 0 1 0>}
   B=MaterialFunctionCall(
    Texture=TextureObject() {samplerType=SAMPLERTYPE_Normal, texture=/Game/Effects/Fort_Effects/Textures/Snowflake/T_Snowflakes_G}
    SubImages=AppendVector(A=@n29, B=@n30)
    Frame=Round(None=@n31.Param4)
   ) {materialFunction=&SubUV_Function}.RGB
   Alpha=Ceil(None=ComponentMask(None=@n32) {g=true})
  )
  n34 = Power(
   Base=MaterialFunctionCall(f2=Abs(None=@n33)) {materialFunction=&AddComponents}.f2
   Exp=Constant() {r=6}
  )
  n35 = Add(A=Multiply(A=Multiply(A=@n34) {constB=10}, B=@n11), B=@n28.RGB)
  n36 = Constant() {r=2}
  n37 = Multiply(A=Multiply(A=LinearInterpolate(A=@n26.Result, B=@n27, Alpha=@n28.R), B=Multiply(A=@n35, B=@n36)), B=Constant() {r=10})
  n38 = Multiply(A=Constant() {r=500}, B=VectorParameter() {parameterName=EmissiveGlow, defaultValue=<rgba 0.0317 0.1963 1 1>}.RGB)
  n39 = Multiply(A=Time()) {constB=0.1}
  n40 = Subtract(A=@n16.Param1) {constB=0.25}
  n41 = TextureSample(UVs=Panner(Coordinate=@n13, Time=@n14, Speed=AppendVector(A=@n39, B=@n40))) {texture=&T_Fire_Mask}
  n42 = Subtract(A=OneMinus(None=Multiply(A=@n41.R)), B=@n18)
  n43 = Add(A=@n37, B=Multiply(A=Multiply(A=@n38, B=@n42), B=DynamicParameter() {defaultValue=<rgba 0 0 1 1>}.Param2))
  MaterialFunctionCall(Emissive=Multiply(A=StaticSwitchParameter(True=@n22.RGB, False=@n23) {parameterName=Use_particle_color}, B=ShadingPathSwitch(Default=@n43, Mobile=@n37))) {materialFunction=&MF_FX_EmissiveAdjustments}
  n44 = TextureObject() {samplerType=SAMPLERTYPE_Normal, texture=&T_Snowflakes_G}
  n45 = AppendVector(A=Constant() {r=2}, B=Constant() {r=2})
  n46 = Round(None=DynamicParameter() {defaultValue=<rgba 0 0 1 1>}.Param4)
  ShadingPathSwitch(
   Default=MaterialFunctionCall(
    Normal=MaterialFunctionCall(Texture=@n44, SubImages=@n45, Frame=@n46) {materialFunction=&SubUV_Function}.RGB
    Flatness=ScalarParameter() {defaultValue=1.48, parameterName=Param_1}
   ) {materialFunction=&FlattenNormal}.Result
   Mobile=Constant3Vector() {constant=<rgba 0 0 1 0>}
  )
 ]
}
